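# Sound cue named for the heavy vocal layer of TeriyakiFishToonMale pickaxe impacts on enemies in Athena.
# Its node graph is empty with no root, so the file shows no wave or playback logic.
# It sets player pickaxe impact attenuation, the impacts sound class, player melee weapon concurrency, a reduced volume and priming on load.
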SoundCue /Game/Athena/Sounds/Weapons/PickAxes/TeriyakiFishToonMale/PA_ImpactEnemy_TeriyakiFishToonMale_Vox_Hv_Cue
root=null
nodes []
properties {
 attenuationSettings=/Game/Sounds/Attenuation/Fort_Weapon_Pickaxe_Player_Impact_Att
 bPrimeOnLoad=true
 maxDistance=3512
 soundClassObject=&Fort_Impacts
 volumeMultiplier=0.225
 concurrencySet [&Player_Melee_Weapon_Conc]
}
